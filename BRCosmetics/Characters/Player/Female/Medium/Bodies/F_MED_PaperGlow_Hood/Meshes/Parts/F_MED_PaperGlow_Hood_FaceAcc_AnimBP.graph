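# Animation Blueprint for the face accessory part of the F_MED_PaperGlow_Hood character, following the character's pose and correcting the hood.
# It copies the attached parent's pose, layering in a Body part pose at clavicle and upper-arm branches; Body is cached when customization loads.
# An additive offset nudges the front hood corrective bone, then a Control Rig refines the result using crouching, parachute and targeting states.
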AnimBlueprint /BRCosmetics/Characters/Player/Female/Medium/Bodies/F_MED_PaperGlow_Hood/Meshes/Parts/F_MED_PaperGlow_Hood_FaceAcc_AnimBP
target_skeleton=&/F_MED_PaperGlow_Hood_FaceAcc_Skeleton
graphs [
 AnimGraph {nodes [
   n1 = Layered blend per bone(
    BasePose=Copy Pose From Mesh() {bUseAttachedParent=true}
    BlendPoses_0=Copy Pose From Mesh(SourceMeshComponent=Get Body() -> (Body:Skeletal Mesh Component Budgeted Object Reference))
   ) {layerSetup [{branchFilters [
       {blendDepth=2, boneName=clavicle_r}
       {blendDepth=2, boneName=clavicle_l}
       {blendDepth=2, boneName=upperarm_l}
       {blendDepth=2, boneName=upperarm_r}
      ]}]}
   n2 = "Transform (Modify) Bone - Bone: Dyn_hood_correct_front"!(ComponentPose=Local To Component(LocalPose=@n1), bAlphaBoolEnabled:bool=True, Translation="5,0,5", Rotation="0,0,0", Scale="1,1,1") -> (Pose) {
    alphaInputType=Bool
    rotationMode=BMM_Additive
    rotationSpace=BCS_BoneSpace
    scaleMode=BMM_Additive
    translationMode=BMM_Additive
    translationSpace=BCS_BoneSpace
    alphaBoolBlend {blendOption=Linear}
    boneToModify {boneName=Dyn_hood_correct_front}
    translation {x=5, z=5}
   }
   n3 = Control Rig!(
    Source=Component To Local(ComponentPose=@n2)
    Alpha
    IsCrouching:bool=Get bIsCrouching()
    IsParachuteOpened:bool=Get bIsParachuteOpened()
    IsTargeting:bool=Get bIsTargeting()
   ) -> (Pose) {
    alphaBoolBlend {blendOption=Linear}
    controlRigAssetReference {blueprintRigClass=&/F_MED_PaperGlow_Hood_FaceAcc_CtrlRig.F_MED_PaperGlow_Hood_FaceAcc_CtrlRig_C}
   }
   Output Pose(Result=@n3) {layerGroup=None}
  ]}
 EventGraph {nodes [
   n1 = Event OnCharacterCustomizationLoaded()
   > Set Body(Body:Skeletal Mesh Component Budgeted Object Reference=Get Array Item with EFort Custom Part Type(Array=@n1.CustomMeshParts)) -> (Output_Get:Skeletal Mesh Component Budgeted Object Reference)
   Event BlueprintUpdateAnimation()
  ]}
]
variables [Body {type=SkeletalMeshComponentBudgeted}]
